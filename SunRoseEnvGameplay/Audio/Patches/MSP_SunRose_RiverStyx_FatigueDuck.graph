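# MetaSound patch for River Styx audio that outputs a volume multiplier for gradual fatigue ducking.
# Start runs a trigger delay of the fatigue threshold, then attacks an ADSR envelope over the fade time, holding at full sustain.
# The envelope scales the duck amount, which is subtracted from unity to form the multiplier.
# Reset releases the envelope over its own fade time but is not wired to the pending delay.
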MetaSoundPatch /SunRoseEnvGameplay/Audio/Patches/MSP_SunRose_RiverStyx_FatigueDuck
inputs [
 FatigueDuckAmount {type=Float}
 FatigueFadeTime {type=Time}
 FatigueResetFadeTime {type=Time}
 FatigueThreshold {type=Time}
 Reset {type=Trigger}
 Start {type=Trigger}
]
outputs [VolumeMultiplier {type=Float}]
pages [Default {
  input_defaults {FatigueDuckAmount=0.2, FatigueFadeTime=15, FatigueResetFadeTime=5, FatigueThreshold=30, Reset=null, Start=null}
  nodes [
   TriggerDelay_0 = UE.Trigger Delay(Delay Time=^FatigueThreshold, In=^Start)
   ADSREnvelopeADSREnvelope_0 = ADSR Envelope.ADSR Envelope:Float(Attack Time=^FatigueFadeTime, Release Time=^FatigueResetFadeTime, Trigger Release=^Reset, Trigger Attack=@TriggerDelay_0.Out) {also {defaults {Decay Time=0, Sustain Level=1}}}
   Add_0 = UE.Add:Float(PrimaryOperand=UE.Multiply:Float(
     AdditionalOperands=@ADSREnvelopeADSREnvelope_0.Out Envelope
     PrimaryOperand=UE.Multiply:Float(PrimaryOperand=^FatigueDuckAmount) {also {defaults {AdditionalOperands=-1}}}.Out
    ).Out) {also {defaults {AdditionalOperands=1}}}
  ]
  links [{dst="@VolumeMultiplier", src=Add_0.Out}]
 }]
