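# MetaSound patch for SunRose River Styx audio that, on each Next trigger, plays a random mono wave from an input array, avoiding recent repeats.
# The random pick triggers a symmetric random pitch shift within PitchSpread, which then starts the Wave Player.
# Playback start rolls a random gain between MinVolume and MaxVolume that scales the output.
# The player's finish trigger drives the OneShot OnFinished output.
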MetaSoundPatch /SunRoseEnvGameplay/Audio/Patches/MSP_SunRose_RiverStyx_PlayRandomWaveFromArray
inputs [
 In Array {type=WaveAsset:Array}
 MaxVolume {type=Float}
 MinVolume {type=Float}
 Next {type=Trigger}
 No Repeats {type=Int32}
 PitchSpread {type=Float}
]
interfaces [UE.Source.OneShot:1.0]
outputs [
 Output {type=Audio}
 UE.Source.OneShot.OnFinished {type=Trigger}
]
pages [Default {
  input_defaults {MaxVolume=0.8, MinVolume=0.5, Next=null, No Repeats=5, PitchSpread=0, In Array []}
  nodes [
   ArrayRandomGet_0 = Array.Random Get:WaveAsset:Array(In Array=^In Array, Next=^Next, No Repeats=^No Repeats) {also {defaults {Weights []}}}
   n1 = UE.RandomFloat(
    Max=UE.Add:Float(PrimaryOperand=^PitchSpread).Out
    Next=@ArrayRandomGet_0.On Next
    Min=UE.Subtract:Float(AdditionalOperands=^PitchSpread).Out
   )
   WavePlayer_0 = UE.Wave Player:Mono(Wave Asset=@ArrayRandomGet_0.Value, Play=@n1.On Next, Pitch Shift=@n1.Value)
   Multiply_0 = UE.Multiply:Audio by Float(AdditionalOperands=UE.RandomFloat(Max=^MaxVolume, Min=^MinVolume, Next=@WavePlayer_0.On Play).Value, PrimaryOperand=@WavePlayer_0.Out Mono)
  ]
  links [{dst="@Output", src=Multiply_0.Out}, {dst="@UE.Source.OneShot.OnFinished", src=WavePlayer_0.On Finished}]
 }]
